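; Particle emitter Blueprint that, judging by its name, shows rain drops on the camera, with a vignette element tied to rendering detail mode.
; On BeginPlay it sets a low SpawnRate on its particle system, then checks the detail mode.
; Above low detail it zeroes VignettSpawnScale.
; At low detail it enables that scale, prints a debug message and plays a timeline that drives the particle Alpha parameter.
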(event EventBeginPlay
  (bind _particlesystemcomponent (Variables|Emitter|GetParticleSystemComponent))
  (Effects|Components|ParticleSystem|SetFloatParameter (Variables|Emitter|GetParticleSystemComponent) "SpawnRate" 0.1)
  (if (>= (Rendering|GetRenderingDetailMode) 1)
    (Effects|Components|ParticleSystem|SetFloatParameter _particlesystemcomponent "VignettSpawnScale" 0.0)
    (else
      (Development|PrintString "LOW Detail!")
      (Effects|Components|ParticleSystem|SetFloatParameter _particlesystemcomponent "VignettSpawnScale" 1.0)
      (bind (_direction _alpha) (Timeline|Alpha PlayFromStart)
        (:Update
          (Effects|Components|ParticleSystem|SetFloatParameter _particlesystemcomponent "Alpha" _alpha))))))

(fn ConstructionScript)
