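; B_Rift_Portals is an enemy spawn rift handling portal vortex, light, audio and wind effects, plus alarms on nearby cars when AI spawns.
; Construction randomizes the vortex and creates dynamic materials, including wind-vector meshes when wind capture is enabled.
; Lifecycle events fade it in or out through material timestamps and light timelines, usually after random delays.
; Lightning and suck-in events add light boosts, quality-gated decal and shake, and wind bursts.
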(event EventBeginPlay
  (bind _returnvalue (Fortnite|GetWindManager))
  (bind _aswind_manager (Utilities|Casting|CastToWindManager _returnvalue))
  ; Add the Wind Vortex to the Wind Manager's render list
  (if (Variables|Default|GetIs%20The%20Wind/Water%20Scene%20Capture%20Activated)
    (Class|WindManager|AddWindComponent _aswind_manager (Variables|Default|GetWindVectorVortex))
    (Class|WindManager|AddWindComponent _aswind_manager (Variables|Default|GetWind%20Vector%20Explosion)))
  ;[1 1 1] Begin Play
  (Rendering|SetVisibility (Variables|RiftVortex|GetRiftVortex) false false)
  (if (Variables|Effects|GetLightEnabled)
    (Rendering|Components|Light|SetIntensity (Variables|Default|GetPointLight) 0.0)
    (else
      (Components|DestroyComponent (Variables|Default|GetPointLight)))))

(event OnComponentBeginOverlap(Sphere)
  (params (OverlappedComponent :type Primitive%20Component%20Object%20Reference) (OtherActor :type Actor%20Object%20Reference) (OtherComp :type Primitive%20Component%20Object%20Reference) (OtherBodyIndex :type Integer) (bFromSweep :type Boolean) (SweepResult :type Hit%20Result%20Structure%20%28by%20ref%29))
  (if (not (Audio|Components|Audio|IsPlaying (Variables|B_Rift_Portals|GetEnemy%20Vortex%20Audio)))
    (Audio|Components|Audio|FadeIn (Variables|B_Rift_Portals|GetEnemy%20Vortex%20Audio) 0.25 1.0 0.0 "Linear" :!class /Script/Engine.AudioComponent)))

(event BuildingRift|EventOnPlaced
  (bind _outactors (Collision|SphereOverlapActors (Transformation|GetActorLocation) 1000 (Utilities|Array|MakeArray "ObjectTypeQuery1") "/Game/Building/ActorBlueprints/Prop/Car_Copper.Car_Copper_C"))
  (bind _output_get (Variables|Default|SetOverlapping%20Actors _outactors))
  (for _array_element _output_get
    (bind _ascar_copper (Utilities|Casting|CastToCar_Copper _array_element))
    (Utilities|Array|Add (Variables|Default|GetArrayOfNearbyCars) _ascar_copper))
  (Utilities|FlowControl|DoOnce false
    (:Completed
      (Utilities|FlowControl|Delay (Math|Random|RandomFloatinRange 1 6))
      (Rendering|SetVisibility (Variables|RiftVortex|GetRiftVortex) true false)
      (Components|Activation|SetActive (Variables|Default|GetP_Rift_Active) true false)
      (Rendering|Material|SetScalarParameterValue (Variables|Effects|GetParticlesBaseMID) "TimestampA" (Utilities|Time|GetGameTimeinSeconds) :!class /Script/Engine.MaterialInstanceDynamic)
      (Rendering|Material|SetScalarParameterValue (Variables|RiftVortex|GetRiftVortexMID) "TimestampAFade" (Utilities|Time|GetGameTimeinSeconds) :!class /Script/Engine.MaterialInstanceDynamic)
      (Rendering|Material|SetScalarParameterValue (Variables|RiftVortex|GetRiftVortexMID) "TransitionLength_InFade" (Variables|Effects|GetLight%20Fade%20In%20Time) :!class /Script/Engine.MaterialInstanceDynamic)
      (if (Variables|Effects|GetLightEnabled)
        (CallFunction|FadeLightIn))
      ;[0.2661 0 1] Ramp up the vortex wind
      (if (Variables|Default|GetIs%20The%20Wind/Water%20Scene%20Capture%20Activated)
        (FortFX|AddMIDAnimationFloat (Variables|Default|GetWind%20Vortex%20MID) "Opacity" 0.0 1 0 false 5)))))

(event OnComponentEndOverlap(Sphere)
  (params (OverlappedComponent :type Primitive%20Component%20Object%20Reference) (OtherActor :type Actor%20Object%20Reference) (OtherComp :type Primitive%20Component%20Object%20Reference) (OtherBodyIndex :type Integer))
  (if (Audio|Components|Audio|IsPlaying (Variables|B_Rift_Portals|GetEnemy%20Vortex%20Audio))
    (Audio|Components|Audio|FadeOut (Variables|B_Rift_Portals|GetEnemy%20Vortex%20Audio) 0.5 0.0 "Linear" :!class /Script/Engine.AudioComponent)))

(event BuildingRift|EventOnActivated
  (Utilities|FlowControl|DoOnce false
    (:Completed
      (Utilities|FlowControl|Delay (Math|Random|RandomFloatinRange 1 6))
      (Rendering|SetVisibility (Variables|RiftVortex|GetRiftVortex) true false)
      (Components|Activation|SetActive (Variables|Default|GetP_Rift_Active) true false)
      (Rendering|Material|SetScalarParameterValue (Variables|Effects|GetParticlesBaseMID) "TimestampA" (Utilities|Time|GetGameTimeinSeconds) :!class /Script/Engine.MaterialInstanceDynamic)
      (Rendering|Material|SetScalarParameterValue (Variables|RiftVortex|GetRiftVortexMID) "TimestampAFade" (Utilities|Time|GetGameTimeinSeconds) :!class /Script/Engine.MaterialInstanceDynamic)
      (Rendering|Material|SetScalarParameterValue (Variables|RiftVortex|GetRiftVortexMID) "TransitionLength_InFade" (Variables|Effects|GetLight%20Fade%20In%20Time) :!class /Script/Engine.MaterialInstanceDynamic)
      (if (Variables|Effects|GetLightEnabled)
        (CallFunction|FadeLightIn))
      ;[0.2661 0 1] Ramp up the vortex wind
      (if (Variables|Default|GetIs%20The%20Wind/Water%20Scene%20Capture%20Activated)
        (FortFX|AddMIDAnimationFloat (Variables|Default|GetWind%20Vortex%20MID) "Opacity" 0.0 1 0 false 5)))))

(event BuildingRift|EventOnSpawnAI
  (for _array_element (Variables|Default|GetArrayOfNearbyCars)
    (Utilities|IsValid _array_element
      (:Is%20Valid
        (Class|CarCopper|FireOffCarAlarm _array_element)))))

(event Custom|ResetGate
  (Utilities|FlowControl|DoOnce false
    (:Completed
      (Utilities|FlowControl|Delay (Math|Random|RandomFloatinRange 1 6))
      (Rendering|SetVisibility (Variables|RiftVortex|GetRiftVortex) true false)
      (Components|Activation|SetActive (Variables|Default|GetP_Rift_Active) true false)
      (Rendering|Material|SetScalarParameterValue (Variables|Effects|GetParticlesBaseMID) "TimestampA" (Utilities|Time|GetGameTimeinSeconds) :!class /Script/Engine.MaterialInstanceDynamic)
      (Rendering|Material|SetScalarParameterValue (Variables|RiftVortex|GetRiftVortexMID) "TimestampAFade" (Utilities|Time|GetGameTimeinSeconds) :!class /Script/Engine.MaterialInstanceDynamic)
      (Rendering|Material|SetScalarParameterValue (Variables|RiftVortex|GetRiftVortexMID) "TransitionLength_InFade" (Variables|Effects|GetLight%20Fade%20In%20Time) :!class /Script/Engine.MaterialInstanceDynamic)
      (if (Variables|Effects|GetLightEnabled)
        (CallFunction|FadeLightIn))
      ;[0.2661 0 1] Ramp up the vortex wind
      (if (Variables|Default|GetIs%20The%20Wind/Water%20Scene%20Capture%20Activated)
        (FortFX|AddMIDAnimationFloat (Variables|Default|GetWind%20Vortex%20MID) "Opacity" 0.0 1 0 false 5)))))

(event Custom|ManuallyActivateRift
  (Rendering|SetVisibility (Variables|RiftVortex|GetRiftVortex) true false)
  (Components|Activation|SetActive (Variables|Default|GetP_Rift_Active) true false)
  (Rendering|Material|SetScalarParameterValue (Variables|Effects|GetParticlesBaseMID) "TimestampA" (Utilities|Time|GetGameTimeinSeconds) :!class /Script/Engine.MaterialInstanceDynamic)
  (Rendering|Material|SetScalarParameterValue (Variables|RiftVortex|GetRiftVortexMID) "TimestampAFade" (Utilities|Time|GetGameTimeinSeconds) :!class /Script/Engine.MaterialInstanceDynamic)
  (Rendering|Material|SetScalarParameterValue (Variables|RiftVortex|GetRiftVortexMID) "TransitionLength_InFade" (Variables|Effects|GetLight%20Fade%20In%20Time) :!class /Script/Engine.MaterialInstanceDynamic)
  (if (Variables|Effects|GetLightEnabled)
    (CallFunction|FadeLightIn))
  ;[0.2661 0 1] Ramp up the vortex wind
  (if (Variables|Default|GetIs%20The%20Wind/Water%20Scene%20Capture%20Activated)
    (FortFX|AddMIDAnimationFloat (Variables|Default|GetWind%20Vortex%20MID) "Opacity" 0.0 1 0 false 5)))

(event Custom|FadeLightIn
  ;[1 1 1] Fade Light
  (Components|Timeline|SetPlayRate (Variables|B_Rift_Portals|GetScaleLightOnOff) (/ 1 (Variables|Effects|GetLight%20Fade%20In%20Time)))
  (bind (_direction _lerp) (Timeline|ScaleLightOnOff PlayFromStart)
    (:Update
      (Rendering|Components|Light|SetIntensity (Variables|Default|GetPointLight) (Math|Float|Lerp 0.0 (Variables|Effects|GetLight%20Intensity%20Idle) _lerp)))))

(event BuildingRift|EventOnStartedEncounterSpawn
  (params (SpawnedAI :type Fort%20AIPawn%20Object%20Reference))
  (bind _rootcomponent (Variables|/Script/Engine.Actor|GetRootComponent SpawnedAI))
  (bind _returnvalue (Transformation|GetWorldLocation _rootcomponent))
  ;[1 1 1] TriggerSpawnSFX
  (|CallParentFunction SpawnedAI)
  (Audio|PlaySoundatLocation "/Game/Sounds/Fort_GamePlay_Sounds/Portal/EnemyPortal_Spawn_Cue.EnemyPortal_Spawn_Cue" _returnvalue)
  (Audio|PlaySoundatLocation "/Game/Sounds/Fort_Monster_Husk/SFX/husk_spawn_yells_Cue.husk_spawn_yells_Cue" _returnvalue))

(event Custom|FadeLightOut
  ;[1 1 1] Fade Light
  (Components|Timeline|SetPlayRate (Variables|B_Rift_Portals|GetScaleLightOnOff) (/ 1 (Variables|Effects|GetLight%20Fade%20Out%20Time)))
  (Timeline|ScaleLightOnOff ReverseFromEnd))

(event Custom|Lightning%20Impact%20Event
  (bind _returnvalue (Utilities|Time|GetGameTimeinSeconds))
  (bind _model_height (Variables|RiftVortex|GetModel%20Height))
  (bind _returnvalue_2 (/ _model_height 2))
  (bind _returnvalue_3 (Math|Vector|MakeVector 0.0 0.0 _returnvalue_2))
  (bind _explosion_wind_vector_mesh_mid (Variables|Default|GetExplosion%20Wind%20Vector%20Mesh%20MID))
  (bind _impactlocation (Variables|Default|GetImpactLocation))
  (bind _pointlight (Variables|Default|GetPointLight))
  (bind _p_rift_randomlightning_01 (Variables|Default|GetP_Rift_RandomLightning_01))
  ;[1 1 1] Struck by lightning
  ;;[1 1 1] Spawn lightning impact effect
  (Networking|CanExecuteCosmeticEvents
    (:True
      (Variables|Default|SetImpactLocation (Transformation|GetActorLocation))
      (Utilities|FlowControl|Sequence
        (:then_0
          (if (Variables|Effects|GetLightEnabled)
            (Components|Timeline|SetPlayRate (Variables|B_Rift_Portals|GetBoostLight) (/ 1 (Variables|Default|GetRiftLightningDuration)))
            ;;;[1 1 1] Boost the light when struck
            (bind (_direction _lerp) (Timeline|BoostLight PlayFromStart)
              (:Update
                (Rendering|Components|Light|SetLightColor _pointlight (Math|Color|Lerp(LinearColor) (Variables|Effects|GetLight%20Color%20Idle) (Variables|Effects|GetLight%20Color%20Active) _lerp) true)
                (Rendering|Components|Light|SetIntensity _pointlight (Math|Float|Lerp (Variables|Effects|GetLight%20Intensity%20Idle) (Variables|Effects|GetLight%20Intensity%20Active) _lerp))
                (Rendering|Lighting|SetAttenuationRadius _pointlight (Math|Float|Lerp (Variables|Effects|GetLight%20Radius%20Idle) (Variables|Effects|GetLight%20Radius%20Active) _lerp))))))
        (:then_1
          (Utilities|FlowControl|Delay 0.15)
          (Utilities|FlowControl|Sequence
            (:then_0
              (if (Variables|Default|GetSpawnImpactParticles)
                (Effects|SpawnEmitteratLocation (Variables|Effects|GetPlasma%20Particles) (Variables|Default|GetImpactLocation))
                (Audio|PlaySoundatLocation "/Game/Sounds/Fort_GamePlay_Sounds/Portal/EnemyPortal_Spawn_Cue.EnemyPortal_Spawn_Cue" _impactlocation)
                (Audio|PlaySoundatLocation "/Game/Sounds/Fort_Monster_Husk/SFX/husk_spawn_yells_Cue.husk_spawn_yells_Cue" _impactlocation)
                (bind _returnvalue_4 (Effects|SpawnEmitteratLocation (Variables|Default|GetLightning%20Strike%20Particles) (Variables|Default|GetImpactLocation)))
                (Utilities|EffectsQualityCheck 0
                  (:Cinematic
                    _)
                  (:Epic
                    _)
                  (:High
                    (bind _returnvalue_5 (Rendering|Decal|SpawnDecalatLocation (Variables|Default|GetDecal%20Material) (Math|Vector|MakeVector 16 (Math|Random|RandomFloatinRange 220 260) (Math|Random|RandomFloatinRange 220 260)) (Variables|Default|GetImpactLocation) "-90.000000,0.000000,0.000000" 10))
                    (Variables|Default|SetDecal _returnvalue_5)
                    (Transformation|AddWorldRotation (Variables|Default|GetDecal) (Math|Rotator|MakeRotator 0.0 0.0 (Math|Random|RandomFloatinRange 0.0 360)) false false)
                    (bind _returnvalue_6 (Rendering|Components|Decal|CreateDynamicMaterialInstance (Variables|Default|GetDecal)))
                    (Rendering|Material|SetScalarParameterValue _returnvalue_6 "Spawntime" (Utilities|Time|GetGameTimeinSeconds) :!class /Script/Engine.MaterialInstanceDynamic)
                    (Camera|PlayWorldCameraShake "/Game/VisualThreatManager/Rifts/Blueprints/B_RiftImpact_CameraShake.B_RiftImpact_CameraShake_C" (Variables|Default|GetImpactLocation) 800 1800))
                  (:Medium
                    _)
                  (:Low
                    _))
                (Rendering|SetCastShadow _returnvalue_4 true)))
            (:then_1
              ;;[1 1 1] From B_Rift_Vortex
              (Transformation|SetRelativeLocation _p_rift_randomlightning_01 _returnvalue_3 false false)
              (Components|Activation|SetActive _p_rift_randomlightning_01 true false)
              (bind _output_get (Variables|RiftVortex|SetLightning%20Spawn%20Loc (+ _returnvalue_3 (Transformation|GetWorldLocation (Variables|RiftVortex|GetRiftVortex)))))
              (Rendering|Material|SetVectorParameterValue (Variables|RiftVortex|GetRiftVortexMID) "Lightning Center" (Math|Conversions|ToLinearColor(Vector) _output_get) :!class /Script/Engine.MaterialInstanceDynamic)
              (Rendering|Material|SetScalarParameterValue (Variables|RiftVortex|GetRiftVortexMID) "TimestampA" _returnvalue :!class /Script/Engine.MaterialInstanceDynamic)
              (Rendering|Material|SetScalarParameterValue (Variables|RiftVortex|GetRiftVortexMID) "TimestampB" (+ _returnvalue (Utilities|ClampMinFloat (- (Variables|Default|GetRiftLightningDuration) (+ (Variables|RiftVortex|GetRiftVortexFadeInTime) (Variables|RiftVortex|GetRiftVortexFadeOutTime))))) :!class /Script/Engine.MaterialInstanceDynamic)
              (Rendering|Material|SetScalarParameterValue (Variables|RiftVortex|GetRiftVortexMID) "TimestampLengthIn" (Variables|RiftVortex|GetRiftVortexFadeInTime) :!class /Script/Engine.MaterialInstanceDynamic)
              (Rendering|Material|SetScalarParameterValue (Variables|RiftVortex|GetRiftVortexMID) "TimestampLengthOut" (Variables|RiftVortex|GetRiftVortexFadeOutTime) :!class /Script/Engine.MaterialInstanceDynamic)
              (Rendering|Material|SetScalarParameterValue (Variables|RiftVortex|GetRiftVortexMID) "Bloat" (Variables|RiftVortex|GetLightning%20Bloat) :!class /Script/Engine.MaterialInstanceDynamic)
              (Utilities|FlowControl|Delay (Variables|Default|GetRiftLightningDuration))
              (Components|Activation|SetActive (Variables|Default|GetP_Rift_RandomLightning_01) false false))
            (:then_2
              ;;[0.2661 0 1] Wind Explosion
              (if (Variables|Default|GetIs%20The%20Wind/Water%20Scene%20Capture%20Activated)
                (FortFX|AddMIDAnimationFloat _explosion_wind_vector_mesh_mid "Opacity" 0.0 2 "/Game/WaterAndWind/Curves/WindExplosionOpacity.WindExplosionOpacity" false 1)
                (FortFX|AddMIDAnimationFloat _explosion_wind_vector_mesh_mid "Explosion Radius" 0.0 768 "/Game/WaterAndWind/Curves/WindExplosionGrowthCurve.WindExplosionGrowthCurve" false 1)))))))))

(event Custom|PatrolEnemySuckIn
  (params (Delay :type Float%20%28double-precision%29) (RiftLightningDuration :type Float%20%28double-precision%29) (SpawnImpactParticles :type Boolean))
  (bind _returnvalue (Utilities|Time|GetGameTimeinSeconds))
  (bind _model_height (Variables|RiftVortex|GetModel%20Height))
  (bind _returnvalue_2 (/ _model_height 2))
  (bind _returnvalue_3 (Math|Vector|MakeVector 0.0 0.0 _returnvalue_2))
  (bind _explosion_wind_vector_mesh_mid (Variables|Default|GetExplosion%20Wind%20Vector%20Mesh%20MID))
  (bind _impactlocation (Variables|Default|GetImpactLocation))
  (bind _pointlight (Variables|Default|GetPointLight))
  (bind _p_rift_randomlightning_01 (Variables|Default|GetP_Rift_RandomLightning_01))
  ;[1 1 1] Struck by lightning
  ;;[1 1 1] Spawn lightning impact effect
  (Utilities|FlowControl|Delay Delay)
  (Variables|Default|SetRiftLightningDuration RiftLightningDuration)
  (Variables|Default|SetSpawnImpactParticles SpawnImpactParticles)
  (Networking|CanExecuteCosmeticEvents
    (:True
      (Variables|Default|SetImpactLocation (Transformation|GetActorLocation))
      (Utilities|FlowControl|Sequence
        (:then_0
          (if (Variables|Effects|GetLightEnabled)
            (Components|Timeline|SetPlayRate (Variables|B_Rift_Portals|GetBoostLight) (/ 1 (Variables|Default|GetRiftLightningDuration)))
            ;;;[1 1 1] Boost the light when struck
            (Timeline|BoostLight PlayFromStart)))
        (:then_1
          (Utilities|FlowControl|Delay 0.15)
          (Utilities|FlowControl|Sequence
            (:then_0
              (if (Variables|Default|GetSpawnImpactParticles)
                (Effects|SpawnEmitteratLocation (Variables|Effects|GetPlasma%20Particles) (Variables|Default|GetImpactLocation))
                (Audio|PlaySoundatLocation "/Game/Sounds/Fort_GamePlay_Sounds/Portal/EnemyPortal_Spawn_Cue.EnemyPortal_Spawn_Cue" _impactlocation)
                (Audio|PlaySoundatLocation "/Game/Sounds/Fort_Monster_Husk/SFX/husk_spawn_yells_Cue.husk_spawn_yells_Cue" _impactlocation)
                (bind _returnvalue_4 (Effects|SpawnEmitteratLocation (Variables|Default|GetLightning%20Strike%20Particles) (Variables|Default|GetImpactLocation)))
                (Utilities|EffectsQualityCheck 0
                  (:Cinematic
                    _)
                  (:Epic
                    _)
                  (:High
                    (bind _returnvalue_5 (Rendering|Decal|SpawnDecalatLocation (Variables|Default|GetDecal%20Material) (Math|Vector|MakeVector 16 (Math|Random|RandomFloatinRange 220 260) (Math|Random|RandomFloatinRange 220 260)) (Variables|Default|GetImpactLocation) "-90.000000,0.000000,0.000000" 10))
                    (Variables|Default|SetDecal _returnvalue_5)
                    (Transformation|AddWorldRotation (Variables|Default|GetDecal) (Math|Rotator|MakeRotator 0.0 0.0 (Math|Random|RandomFloatinRange 0.0 360)) false false)
                    (bind _returnvalue_6 (Rendering|Components|Decal|CreateDynamicMaterialInstance (Variables|Default|GetDecal)))
                    (Rendering|Material|SetScalarParameterValue _returnvalue_6 "Spawntime" (Utilities|Time|GetGameTimeinSeconds) :!class /Script/Engine.MaterialInstanceDynamic)
                    (Camera|PlayWorldCameraShake "/Game/VisualThreatManager/Rifts/Blueprints/B_RiftImpact_CameraShake.B_RiftImpact_CameraShake_C" (Variables|Default|GetImpactLocation) 800 1800))
                  (:Medium
                    _)
                  (:Low
                    _))
                (Rendering|SetCastShadow _returnvalue_4 true)))
            (:then_1
              ;;[1 1 1] From B_Rift_Vortex
              (Transformation|SetRelativeLocation _p_rift_randomlightning_01 _returnvalue_3 false false)
              (Components|Activation|SetActive _p_rift_randomlightning_01 true false)
              (bind _output_get (Variables|RiftVortex|SetLightning%20Spawn%20Loc (+ _returnvalue_3 (Transformation|GetWorldLocation (Variables|RiftVortex|GetRiftVortex)))))
              (Rendering|Material|SetVectorParameterValue (Variables|RiftVortex|GetRiftVortexMID) "Lightning Center" (Math|Conversions|ToLinearColor(Vector) _output_get) :!class /Script/Engine.MaterialInstanceDynamic)
              (Rendering|Material|SetScalarParameterValue (Variables|RiftVortex|GetRiftVortexMID) "TimestampA" _returnvalue :!class /Script/Engine.MaterialInstanceDynamic)
              (Rendering|Material|SetScalarParameterValue (Variables|RiftVortex|GetRiftVortexMID) "TimestampB" (+ _returnvalue (Utilities|ClampMinFloat (- (Variables|Default|GetRiftLightningDuration) (+ (Variables|RiftVortex|GetRiftVortexFadeInTime) (Variables|RiftVortex|GetRiftVortexFadeOutTime))))) :!class /Script/Engine.MaterialInstanceDynamic)
              (Rendering|Material|SetScalarParameterValue (Variables|RiftVortex|GetRiftVortexMID) "TimestampLengthIn" (Variables|RiftVortex|GetRiftVortexFadeInTime) :!class /Script/Engine.MaterialInstanceDynamic)
              (Rendering|Material|SetScalarParameterValue (Variables|RiftVortex|GetRiftVortexMID) "TimestampLengthOut" (Variables|RiftVortex|GetRiftVortexFadeOutTime) :!class /Script/Engine.MaterialInstanceDynamic)
              (Rendering|Material|SetScalarParameterValue (Variables|RiftVortex|GetRiftVortexMID) "Bloat" (Variables|RiftVortex|GetLightning%20Bloat) :!class /Script/Engine.MaterialInstanceDynamic)
              (Utilities|FlowControl|Delay (Variables|Default|GetRiftLightningDuration))
              (Components|Activation|SetActive (Variables|Default|GetP_Rift_RandomLightning_01) false false))
            (:then_2
              ;;[0.2661 0 1] Wind Explosion
              (if (Variables|Default|GetIs%20The%20Wind/Water%20Scene%20Capture%20Activated)
                (FortFX|AddMIDAnimationFloat _explosion_wind_vector_mesh_mid "Opacity" 0.0 2 "/Game/WaterAndWind/Curves/WindExplosionOpacity.WindExplosionOpacity" false 1)
                (FortFX|AddMIDAnimationFloat _explosion_wind_vector_mesh_mid "Explosion Radius" 0.0 768 "/Game/WaterAndWind/Curves/WindExplosionGrowthCurve.WindExplosionGrowthCurve" false 1)))))))))

(event BuildingRift|EventOnDeactivated
  ;[1 1 1] On Removed
  (Utilities|FlowControl|Delay (Math|Random|RandomFloatinRange 0.0 3))
  (Networking|CanExecuteCosmeticEvents
    (:True
      (Rendering|Material|SetScalarParameterValue (Variables|Effects|GetParticlesBaseMID) "TimestampB" (Utilities|Time|GetGameTimeinSeconds) :!class /Script/Engine.MaterialInstanceDynamic)
      (Rendering|Material|SetScalarParameterValue (Variables|RiftVortex|GetRiftVortexMID) "TimestampBFade" (Utilities|Time|GetGameTimeinSeconds) :!class /Script/Engine.MaterialInstanceDynamic)
      (Rendering|Material|SetScalarParameterValue (Variables|RiftVortex|GetRiftVortexMID) "TransitionLength_OutFade" (Variables|Effects|GetLight%20Fade%20Out%20Time) :!class /Script/Engine.MaterialInstanceDynamic)
      (if (Variables|Effects|GetLightEnabled)
        (CallFunction|FadeLightOut)
        (else
          _))
      (Utilities|FlowControl|Delay (Variables|Effects|GetLight%20Fade%20Out%20Time))
      (Rendering|SetVisibility (Variables|RiftVortex|GetRiftVortex) false false)
      (Components|Activation|SetActive (Variables|Default|GetP_Rift_Active) false false)
      (CallFunction|ResetGate)
      ;[0.2661 0 1] Ramp the vortex wind down
      (if (Variables|Default|GetIs%20The%20Wind/Water%20Scene%20Capture%20Activated)
        (FortFX|AddMIDAnimationFloat (Variables|Default|GetWind%20Vortex%20MID) "Opacity" 1 0.0 0 false 5)))))

(event BuildingRift|EventOnRemoved
  ;[1 1 1] On Removed
  (Utilities|FlowControl|Delay (Math|Random|RandomFloatinRange 0.0 3))
  (Networking|CanExecuteCosmeticEvents
    (:True
      (Rendering|Material|SetScalarParameterValue (Variables|Effects|GetParticlesBaseMID) "TimestampB" (Utilities|Time|GetGameTimeinSeconds) :!class /Script/Engine.MaterialInstanceDynamic)
      (Rendering|Material|SetScalarParameterValue (Variables|RiftVortex|GetRiftVortexMID) "TimestampBFade" (Utilities|Time|GetGameTimeinSeconds) :!class /Script/Engine.MaterialInstanceDynamic)
      (Rendering|Material|SetScalarParameterValue (Variables|RiftVortex|GetRiftVortexMID) "TransitionLength_OutFade" (Variables|Effects|GetLight%20Fade%20Out%20Time) :!class /Script/Engine.MaterialInstanceDynamic)
      (if (Variables|Effects|GetLightEnabled)
        (CallFunction|FadeLightOut)
        (else
          _))
      (Utilities|FlowControl|Delay (Variables|Effects|GetLight%20Fade%20Out%20Time))
      (Rendering|SetVisibility (Variables|RiftVortex|GetRiftVortex) false false)
      (Components|Activation|SetActive (Variables|Default|GetP_Rift_Active) false false)
      (CallFunction|ResetGate)
      ;[0.2661 0 1] Ramp the vortex wind down
      (if (Variables|Default|GetIs%20The%20Wind/Water%20Scene%20Capture%20Activated)
        (FortFX|AddMIDAnimationFloat (Variables|Default|GetWind%20Vortex%20MID) "Opacity" 1 0.0 0 false 5)))))

(event EventEndPlay
  (params (EndPlayReason :type EEndPlayReason%20Enum :default "Destroyed"))
  ;[1 1 1] On Removed
  (Utilities|FlowControl|Delay (Math|Random|RandomFloatinRange 0.0 3))
  (Networking|CanExecuteCosmeticEvents
    (:True
      (Rendering|Material|SetScalarParameterValue (Variables|Effects|GetParticlesBaseMID) "TimestampB" (Utilities|Time|GetGameTimeinSeconds) :!class /Script/Engine.MaterialInstanceDynamic)
      (Rendering|Material|SetScalarParameterValue (Variables|RiftVortex|GetRiftVortexMID) "TimestampBFade" (Utilities|Time|GetGameTimeinSeconds) :!class /Script/Engine.MaterialInstanceDynamic)
      (Rendering|Material|SetScalarParameterValue (Variables|RiftVortex|GetRiftVortexMID) "TransitionLength_OutFade" (Variables|Effects|GetLight%20Fade%20Out%20Time) :!class /Script/Engine.MaterialInstanceDynamic)
      (if (Variables|Effects|GetLightEnabled)
        (CallFunction|FadeLightOut)
        (else
          _))
      (Utilities|FlowControl|Delay (Variables|Effects|GetLight%20Fade%20Out%20Time))
      (Rendering|SetVisibility (Variables|RiftVortex|GetRiftVortex) false false)
      (Components|Activation|SetActive (Variables|Default|GetP_Rift_Active) false false)
      (CallFunction|ResetGate)
      ;[0.2661 0 1] Ramp the vortex wind down
      (if (Variables|Default|GetIs%20The%20Wind/Water%20Scene%20Capture%20Activated)
        (FortFX|AddMIDAnimationFloat (Variables|Default|GetWind%20Vortex%20MID) "Opacity" 1 0.0 0 false 5)))))

(event Custom|ManuallyDeactivateRift
  ;[1 1 1] On Removed
  ;;[1 1 1] Manually Remove
  (Utilities|FlowControl|Delay (Variables|Default|GetManualDeactivateDelayBeforeDestory))
  (Utilities|FlowControl|Sequence
    (:then_0
      (Networking|CanExecuteCosmeticEvents
        (:True
          (Rendering|Material|SetScalarParameterValue (Variables|Effects|GetParticlesBaseMID) "TimestampB" (Utilities|Time|GetGameTimeinSeconds) :!class /Script/Engine.MaterialInstanceDynamic)
          (Rendering|Material|SetScalarParameterValue (Variables|RiftVortex|GetRiftVortexMID) "TimestampBFade" (Utilities|Time|GetGameTimeinSeconds) :!class /Script/Engine.MaterialInstanceDynamic)
          (Rendering|Material|SetScalarParameterValue (Variables|RiftVortex|GetRiftVortexMID) "TransitionLength_OutFade" (Variables|Effects|GetLight%20Fade%20Out%20Time) :!class /Script/Engine.MaterialInstanceDynamic)
          (if (Variables|Effects|GetLightEnabled)
            (CallFunction|FadeLightOut)
            (else
              _))
          (Utilities|FlowControl|Delay (Variables|Effects|GetLight%20Fade%20Out%20Time))
          (Rendering|SetVisibility (Variables|RiftVortex|GetRiftVortex) false false)
          (Components|Activation|SetActive (Variables|Default|GetP_Rift_Active) false false)
          (CallFunction|ResetGate)
          ;[0.2661 0 1] Ramp the vortex wind down
          (if (Variables|Default|GetIs%20The%20Wind/Water%20Scene%20Capture%20Activated)
            (FortFX|AddMIDAnimationFloat (Variables|Default|GetWind%20Vortex%20MID) "Opacity" 1 0.0 0 false 5)))))
    (:then_1
      ;;[1 1 1] Only check for "can execute cosmetic events" in the first sequence pin so the server can destroy the actor
      (Utilities|FlowControl|Delay 5)
      (Actor|DestroyActor))))

(fn ConstructionScript
  (bind (_x _y) (Math|Vector2D|BreakVector2D (Utilities|RandomVector2dinRange (Variables|RiftVortex|GetWiggle%20Speed%20Min) (Variables|RiftVortex|GetWiggle%20Speed%20Max))))
  (bind _uv_tile_multiplier_max (Variables|RiftVortex|GetUV%20Tile%20Multiplier%20Max))
  (bind _uv_tile_multiplier_min (Variables|RiftVortex|GetUV%20Tile%20Multiplier%20Min))
  (bind _vector2dout (Utilities|RandomVector2dinRange _uv_tile_multiplier_min _uv_tile_multiplier_max))
  (bind (_x_2 _y_2) (Math|Vector2D|BreakVector2D _vector2dout))
  (bind _pointlight (Variables|Default|GetPointLight))
  (bind _windvectorvortex (Variables|Default|GetWindVectorVortex))
  (bind _wind_vector_explosion (Variables|Default|GetWind%20Vector%20Explosion))
  (Networking|CanExecuteCosmeticEvents
    (:True
      (Utilities|FlowControl|Sequence
        (:then_0
          ;[1 1 1] Original construction script of B_Rift_Portals
          (bind _returnvalue (Rendering|Material|CreateDynamicMaterialInstance (Variables|Default|GetP_Rift_Active) 0 0 "None" :!class /Script/Engine.PrimitiveComponent))
          (Variables|Effects|SetParticlesBaseMID _returnvalue)
          (Utilities|EffectsQualityCheck 0
            (:Cinematic
              _)
            (:Epic
              _)
            (:High
              _)
            (:Medium
              (Rendering|SetVisibility (Variables|Default|GetPointLight) false false)
              (Variables|Effects|SetLightEnabled false))
            (:Low
              _))
          (Variables|Effects|SetLight%20Intensity%20Idle (Variables|/Script/Engine.LightComponentBase|GetIntensity _pointlight))
          (Variables|Effects|SetLight%20Radius%20Idle (Variables|/Script/Engine.PointLightComponent|GetAttenuationRadius _pointlight))
          (Variables|Effects|SetLight%20Color%20Idle (Rendering|Components|Light|GetLightColor _pointlight))
          (Variables|Effects|SetLightEnabled true))
        (:then_1
          ;[1 1 1] Construction script from B_RiftVortex_01
          (Variables|RiftVortex|SetVortexSize (Math|Vector|MakeVector (.x (Utilities|RandomVectorinRange (Variables|RiftVortex|GetVortexSizeMin) (Variables|RiftVortex|GetVortexSizeMax))) (.x (Utilities|RandomVectorinRange (Variables|RiftVortex|GetVortexSizeMin) (Variables|RiftVortex|GetVortexSizeMax))) (.z (Utilities|RandomVectorinRange (Variables|RiftVortex|GetVortexSizeMin) (Variables|RiftVortex|GetVortexSizeMax)))))
          (Variables|RiftVortex|SetTwist%20Amount (Math|Random|RandomFloatinRange (Variables|RiftVortex|GetTwist%20Amount%20Min) (Variables|RiftVortex|GetTwist%20Amount%20Max)))
          (Variables|RiftVortex|SetWiggle%20Strength (Math|Random|RandomFloatinRange (Variables|RiftVortex|GetWiggle%20Strength%20Min) (Variables|RiftVortex|GetWiggle%20Strength%20Max)))
          ;;[1 1 1] create MIDs
          (bind _returnvalue_2 (Rendering|Material|CreateDynamicMaterialInstance (Variables|RiftVortex|GetRiftVortex) 0 0 "None" :!class /Script/Engine.PrimitiveComponent))
          (Variables|RiftVortex|SetRiftVortexMID _returnvalue_2)
          (Transformation|SetWorldScale3D (Variables|RiftVortex|GetRiftVortex) (Variables|RiftVortex|GetVortexSize))
          (Rendering|Material|SetScalarParameterValue (Variables|RiftVortex|GetRiftVortexMID) "curve along axis amount" (Variables|RiftVortex|GetTwist%20Amount) :!class /Script/Engine.MaterialInstanceDynamic)
          (Rendering|Material|SetScalarParameterValue (Variables|RiftVortex|GetRiftVortexMID) "Axis Offset Amount" (Variables|RiftVortex|GetWiggle%20Strength) :!class /Script/Engine.MaterialInstanceDynamic)
          (Rendering|Material|SetScalarParameterValue (Variables|RiftVortex|GetRiftVortexMID) "WPO Strength" (Variables|RiftVortex|GetWPO%20Strength) :!class /Script/Engine.MaterialInstanceDynamic)
          (Rendering|Material|SetVectorParameterValue (Variables|RiftVortex|GetRiftVortexMID) "Pivot position" (Math|Conversions|ToLinearColor(Vector) (Transformation|GetWorldLocation (Variables|RiftVortex|GetRiftVortex))) :!class /Script/Engine.MaterialInstanceDynamic)
          (Rendering|Material|SetScalarParameterValue (Variables|RiftVortex|GetRiftVortexMID) "Total Model height" (* (.z (Variables|RiftVortex|GetVortexSize)) (Variables|RiftVortex|GetModel%20Height)) :!class /Script/Engine.MaterialInstanceDynamic)
          (Rendering|Material|SetScalarParameterValue (Variables|RiftVortex|GetRiftVortexMID) "Tile Master U" _x_2 :!class /Script/Engine.MaterialInstanceDynamic)
          (Rendering|Material|SetScalarParameterValue (Variables|RiftVortex|GetRiftVortexMID) "Tile Master v" _y_2 :!class /Script/Engine.MaterialInstanceDynamic)
          (Rendering|Material|SetScalarParameterValue (Variables|RiftVortex|GetRiftVortexMID) "PatternTilingFrequency" (Math|Random|RandomFloatinRange (Variables|RiftVortex|GetWiggle%20Frequency%20Min) (Variables|RiftVortex|GetWiggle%20Frequency%20Max)) :!class /Script/Engine.MaterialInstanceDynamic)
          (Rendering|Material|SetVectorParameterValue (Variables|RiftVortex|GetRiftVortexMID) "XY pan speed" (Utilities|Struct|MakeLinearColor _x _y) :!class /Script/Engine.MaterialInstanceDynamic))
        (:then_2
          ;[1 1 1] Wind prep
          (bind _returnvalue_3 (Utilities|GetIsSceneCaptureWindWaterEnabled))
          (bind _output_get (Variables|Default|SetIs%20The%20Wind/Water%20Scene%20Capture%20Activated _returnvalue_3))
          (if _output_get
            (Rendering|SetVisibility _windvectorvortex true false)
            (bind _returnvalue_4 (Rendering|Material|CreateDynamicMaterialInstance _windvectorvortex 0 0 "None" :!class /Script/Engine.PrimitiveComponent))
            (bind _output_get_2 (Variables|Default|SetWind%20Vortex%20MID _returnvalue_4))
            (Rendering|Material|SetScalarParameterValue _output_get_2 "Opacity" 0.0 :!class /Script/Engine.MaterialInstanceDynamic)
            (Rendering|SetVisibility _wind_vector_explosion true false)
            (bind _returnvalue_5 (Rendering|Material|CreateDynamicMaterialInstance _wind_vector_explosion 0 0 "None" :!class /Script/Engine.PrimitiveComponent))
            (bind _output_get_3 (Variables|Default|SetExplosion%20Wind%20Vector%20Mesh%20MID _returnvalue_5))
            (Rendering|Material|SetScalarParameterValue _output_get_3 "Opacity" 0.0 :!class /Script/Engine.MaterialInstanceDynamic)
            (Rendering|Material|SetScalarParameterValue _output_get_3 "Intensity Tweak" 2 :!class /Script/Engine.MaterialInstanceDynamic)))))))
