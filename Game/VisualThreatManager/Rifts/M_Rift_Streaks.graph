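# Translucent, unlit, two-sided mesh-particle material for rift streak effects, compatible with Niagara.
# A circular effect texture pans along U by particle lifetime, warped by normal-map noise, masked along U and depth-faded.
# Its red channel drives particle-colored emissive and opacity, while green drives refraction, with dynamic parameters for lifetime, fadeout, speed and refraction.
# The graph also holds time-based axis rotation and a switchable position offset that reads particle color as world position.
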Material /Game/VisualThreatManager/Rifts/M_Rift_Streaks
graph {
 properties {
  bUsedWithMeshParticles=true
  bUsedWithNiagaraMeshParticles=true
  blendMode=BLEND_Translucent
  pixelDepthOffsetMode=PDOM_Legacy
  refractionMethod=RM_IndexOfRefraction
  shadingModel=MSM_Unlit
  translucencyPass=MTP_BeforeDOF
  twoSided=true
  usedShadingModels=Unlit
 }
 nodes [
  n1 = TextureSample() {texture=/Game/VisualThreatManager/Rifts/Textures/CircularEffectTexture}
  n2 = TextureCoordinate() {uTiling=2}
  n3 = Multiply(A=ComponentMask(None=TextureSampleParameter2D(UVs=@n2) {parameterName=Normal, samplerType=SAMPLERTYPE_Normal, texture=/Game/Global/Textures/Default/Noise/T_Normal_Noise01}.RGB) {g=true, r=true}) {constB=0.2}
  n4 = MaterialFunctionCall(Float2=TextureCoordinate()) {materialFunction=&BreakOutFloat2Components}
  n5 = MaterialFunctionCall!(
   UV Channel=@n4.R
   Tiling Amount=Constant() {r=5}
   "Time (0-1)"=OneMinus(None=DynamicParameter() {defaultValue=<rgba 0.25 0 1 1.1>, paramNames [lifetime 0-1, fadeout, Speed, Refraction]}.lifetime 0-1)
  ) {materialFunction=/Engine/Functions/Engine_MaterialFunctions02/UVs/PanTextureCoordinateChannelfrom-1toN+1.PanTextureCoordinateChannelfrom-1ton+1}
  n6 = TextureSample(UVs=Add(A=Multiply(A=@n1.B, B=@n3), B=AppendVector(A=@n5.UV Channel output, B=@n4.G))) {texture=&CircularEffectTexture}
  LinearInterpolate(
   B=LinearInterpolate(B=DynamicParameter() {defaultValue=<rgba 0.25 0 1 1.1>, paramNames [lifetime 0-1, fadeout, Speed, Refraction]}.Refraction, Alpha=@n6.G) {constA=1}
   Alpha=ParticleColor().A
  ) {constA=1}
  n7 = ConstantBiasScale(None=TextureCoordinate()) {bias=-0.5, scale=2}
  n8 = OneMinus(None=Multiply(A=@n7, B=@n7))
  n9 = Multiply(A=@n8, B=@n8)
  n10 = Multiply(
   A=ComponentMask(None=Multiply(A=@n9) {constB=2}) {r=true}
   B=DepthFade(Opacity=ScalarParameter() {defaultValue=20, parameterName=Depth Fade Distance}) {fadeDistanceDefault=512}
  )
  n11 = Multiply(A=DynamicParameter() {defaultValue=<rgba 0.25 0 1 1.1>, paramNames [lifetime 0-1, fadeout, Speed, Refraction]}.fadeout, B=@n10)
  n12 = ParticleColor()
  MaterialFunctionCall(Emissive=Multiply(A=Multiply(A=@n6.R, B=@n11), B=@n12.RGB)) {materialFunction=&MF_FX_EmissiveAdjustments}
  Multiply(A=@n12.A, B=Multiply(A=@n6.R, B=@n10))
  n13 = VectorParameter() {parameterName=NormalizedRotationAxis, defaultValue=<rgba -0.01 0 1 0>}
  RotateAboutAxis(
   NormalizedRotationAxis=Normalize(VectorInput=Transform(None=@n13.RGB) {transformSourceType=TRANSFORMSOURCE_Local})
   RotationAngle=Multiply(A=DynamicParameter() {defaultValue=<rgba 0.25 0 1 1.1>, paramNames [lifetime 0-1, fadeout, Speed, Refraction]}.Speed, B=Time())
   PivotPoint=ActorPositionWS()
   Position=WorldPosition().XYZ
  )
  n14 = ParticleColor() {desc=XYZ world position Scale}
  n15 = ActorPositionWS()
  n16 = Multiply(A=ComponentMask(None=TextureCoordinate()) {g=true}) {constB=2}
  StaticSwitchParameter(True=Multiply(A=Subtract(A=@n14.RGB, B=@n15), B=Clamp(None=@n16)), False=Constant()) {parameterName=UseWPOToControlPos}
  TextureSampleParameter2D() {parameterName=Normal, samplerType=SAMPLERTYPE_Normal, texture=&T_Normal_Noise01}
 ]
}
