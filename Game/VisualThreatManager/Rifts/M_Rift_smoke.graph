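# Translucent unlit particle material for rift smoke, usable on Niagara and Cascade sprites.
# A smooth smoke mask texture's RGB is multiplied by particle color and passed through the shared MF_FX_EmissiveAdjustments function for the glow color.
# The mask alpha is multiplied by particle alpha and a depth fade, giving a transparency term that softens where sprites meet scene geometry.
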Material /Game/VisualThreatManager/Rifts/M_Rift_smoke.M_Rift_Smoke
graph {
 properties {
  bUsedWithNiagaraSprites=true
  bUsedWithParticleSprites=true
  blendMode=BLEND_Translucent
  pixelDepthOffsetMode=PDOM_Legacy
  shadingModel=MSM_Unlit
  usedShadingModels=Unlit
 }
 nodes [
  n1 = TextureSample() {texture=/Game/Effects/Fort_Effects/Textures/T_Smooth_Smoke_mask}
  n2 = ParticleColor()
  MaterialFunctionCall(Emissive=Multiply(A=@n1.RGB, B=@n2.RGB)) {materialFunction=&MF_FX_EmissiveAdjustments}
  Multiply(A=Multiply(A=@n1.A, B=@n2.A), B=DepthFade() {fadeDistanceDefault=60})
 ]
}
