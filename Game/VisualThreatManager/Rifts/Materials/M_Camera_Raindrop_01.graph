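# Unlit translucent particle sprite material for raindrops on the camera lens, with index-of-refraction distortion.
# A rain gradient, broken up by screen-tiled noise, shapes opacity, while a centred inverted sphere mask fades opacity and refraction toward screen middle.
# The same noise warps sphere-normal UVs; particle colour drives emissive through an edge exponent.
# A static switch exposes detail controls through Niagara dynamic parameters, and Low quality simplifies opacity and emissive.
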Material /Game/VisualThreatManager/Rifts/Materials/M_Camera_Raindrop_01
graph {
 properties {
  bDisableDepthTest=true
  bEnableResponsiveAA=true
  bUsedWithNiagaraSprites=true
  bUsedWithParticleSprites=true
  blendMode=BLEND_Translucent
  pixelDepthOffsetMode=PDOM_Legacy
  refractionMethod=RM_IndexOfRefraction
  shadingModel=MSM_Unlit
  translucencyPass=MTP_BeforeDOF
  usedShadingModels=Unlit
 }
 nodes [
  n1 = ParticleColor()
  n2 = TextureSample() {samplerType=SAMPLERTYPE_Grayscale, texture=/Game/VisualThreatManager/Rifts/Textures/T_RainGrad}
  n3 = DynamicParameter() {parameterIndex=1, defaultValue=<rgba 0.6 8 1 1>, paramNames [ScaleDetailNormal, DetailTile, SimpleDetail, Param4]}
  n4 = StaticSwitchParameter(True=@n3.DetailTile, False=ScalarParameter() {defaultValue=8, parameterName=U}) {parameterName=UseInNiagara?}
  n5 = StaticSwitchParameter(True=@n3.DetailTile, False=ScalarParameter() {defaultValue=8, parameterName=V}) {parameterName=UseInNiagara?}
  n6 = TextureSampleParameter2D(UVs=Multiply(A=ScreenPosition().ViewportUV, B=AppendVector(A=@n4, B=@n5))) {parameterName=NoiseTexture, texture=&T_NoisePack_01}
  n7 = AppendVector(A=@n6.R, B=@n6.R)
  n8 = Multiply(
   A=Add(A=@n7, B=ScalarParameter() {defaultValue=-0.5, parameterName=Bias Opacity})
   B=ScalarParameter() {defaultValue=3, parameterName=Scale Opacity}
  )
  n9 = Multiply(A=@n2.G, B=Add(A=@n2.G, B=Desaturation(None=@n8)))
  n10 = OneMinus(None=SphereMask(
    A=ScreenPosition().ViewportUV
    B=Constant() {r=0.5}
    Radius=ScalarParameter() {defaultValue=0.5, parameterName=FullScreenMaskRadius}
    Hardness=ScalarParameter() {defaultValue=0.6, parameterName=FullScreenMaskHardness}
   ))
  n11 = QualitySwitch(Default=Multiply(A=Multiply(A=@n1.A, B=@n9), B=@n10), Low=Multiply(A=@n1.A, B=@n2.G))
  n12 = DynamicParameter() {parameterIndex=1, defaultValue=<rgba 0.6 8 1 1>, paramNames [ScaleDetailNormal, DetailTile, SimpleDetail, Param4]}
  LinearInterpolate(
   A=@n11
   B=LinearInterpolate(
    A=@n11
    B=TextureSample() {samplerType=SAMPLERTYPE_Grayscale, texture=&T_RainGrad}.G
    Alpha=@n12.SimpleDetail
   )
   Alpha=StaticSwitchParameter(True=Constant() {r=1}, False=Constant()) {parameterName=UseInNiagara?}
  )
  LinearInterpolate(
   B=LinearInterpolate(B=DynamicParameter() {paramNames [Refraction Amount, BiasScale Opacity, BiasScale Normal, Param4]}.Refraction Amount, Alpha=@n2.G) {constA=1}
   Alpha=@n10
  ) {constA=1}
  n13 = LinearInterpolate(A=Constant() {r=1}, B=@n2.G, Alpha=ScalarParameter() {defaultValue=1, parameterName=UseTextureWithEmissive})
  n14 = Add(A=OneMinus(None=@n13), B=@n1.RGB)
  n15 = Power(Base=@n13, Exp=ScalarParameter() {defaultValue=5, parameterName=Edge Exponent})
  MaterialFunctionCall(Emissive=QualitySwitch(Default=Multiply(A=@n14, B=@n15), Low=Multiply(A=@n1.RGB, B=@n13))) {materialFunction=&MF_FX_EmissiveAdjustments}
  Multiply()
  n16 = Add(A=@n7, B=ScalarParameter() {defaultValue=-0.3, parameterName=Bias Normal})
  n17 = DynamicParameter() {parameterIndex=1, defaultValue=<rgba 0.6 8 1 1>, paramNames [ScaleDetailNormal, DetailTile, SimpleDetail, Param4]}
  n18 = StaticSwitchParameter(True=@n17.ScaleDetailNormal, False=ScalarParameter() {defaultValue=0.6, parameterName=Scale Normal}) {parameterName=UseInNiagara?}
  TextureSample(UVs=Add(A=Multiply(A=@n16, B=@n18), B=TextureCoordinate())) {samplerType=SAMPLERTYPE_Normal, texture=&T_Sphere_N}
 ]
}
