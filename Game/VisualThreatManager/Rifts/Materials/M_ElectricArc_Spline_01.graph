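# Translucent unlit electric arc material for spline-thickened meshes, usable on Niagara mesh and ribbon particles.
# UV gradient masks drive color and opacity, with optional core glow, depth-fade contact glow and particle color.
# World position offset combines SplineThicken, widened by camera distance and scrolling noise, with ramp-masked normal-noise wobble on optionally stepped time.
# Dynamic parameters control thickness, randomness, pan speed and growth; mobile drops wobble and contact glow.
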Material /Game/VisualThreatManager/Rifts/Materials/M_ElectricArc_Spline_01
graph {
 properties {
  bUsedWithMeshParticles=true
  bUsedWithNanite=true
  bUsedWithNiagaraMeshParticles=true
  bUsedWithNiagaraRibbons=true
  blendMode=BLEND_Translucent
  pixelDepthOffsetMode=PDOM_Legacy
  shadingModel=MSM_Unlit
  twoSided=true
  usedShadingModels=Unlit
 }
 nodes [
  n1 = ParticleColor()
  n2 = ComponentMask(None=TextureCoordinate()) {r=true}
  n3 = OneMinus(None=@n2)
  n4 = Saturate(None=Multiply(B=Multiply(A=@n3, B=@n2)) {constA=4})
  n5 = Add(
   A=Multiply(A=ScalarParameter() {defaultValue=0.1, group=Color, parameterName=Glow Value}, B=@n4) {constA=0.1}
   B=Power(Base=@n4, Exp=ScalarParameter() {defaultValue=1, group=Color, parameterName=Core Exponent}) {constExponent=3}
  )
  n6 = Constant() {r=1}
  n7 = ComponentMask(None=TextureCoordinate()) {g=true}
  n8 = Multiply(A=OneMinus(None=@n7), B=@n7)
  n9 = ComponentMask(None=TextureCoordinate()) {g=true}
  n10 = Multiply(
   A=Power(
    Base=StaticSwitchParameter(True=@n8, False=@n9) {defaultValue=true, parameterName=GradientRampToggle}
    Exp=ScalarParameter() {defaultValue=3, group=Color, parameterName=Mask Exponent}
   ) {constExponent=3}
   B=ScalarParameter() {defaultValue=50, group=Color, parameterName=Mask Scale}
  ) {constB=50}
  n11 = Clamp(None=Multiply(A=StaticSwitchParameter(True=@n5, False=@n6) {parameterName=CoreGlowToggle}, B=@n10))
  n12 = Multiply(
   A=LinearInterpolate(
    A=VectorParameter() {group=Color, parameterName=Color Outer, defaultValue=<rgba 0.4 0.4 60 1>}.RGB
    B=VectorParameter() {group=Color, parameterName=Color Inner, defaultValue=<rgba 16.95 3.9 30 1>}.RGB
    Alpha=@n11
   )
   B=ScalarParameter() {defaultValue=1, parameterName=Mobile Intensity Multiplier}
  )
  n13 = Multiply(B=@n12) {constA=16}
  n14 = DepthFade(Opacity=@n11, FadeDistance=ScalarParameter() {defaultValue=2, parameterName=ContactGlowDistance})
  n15 = ShadingPathSwitch(Default=StaticSwitchParameter(True=LinearInterpolate(A=@n12, B=@n13, Alpha=@n14), False=@n12) {parameterName=UseContactGlow}, Mobile=@n12) {desc=Migrated from FeatureLevelSwitch}
  MaterialFunctionCall(Emissive=StaticSwitchParameter(True=Multiply(A=@n1.RGB, B=@n15), False=@n15) {parameterName=UseParticleColor}) {materialFunction=&MF_FX_EmissiveAdjustments}
  n16 = DynamicParameter() {defaultValue=<rgba 1 0 1 1>, paramNames [Thickness,Random,PanSpeed,PowerGrow]}
  n17 = ScalarParameter() {defaultValue=2, group=WPO, parameterName=WidthBase}
  n18 = ObjectPositionWS()
  n19 = CameraPositionWS()
  n20 = Clamp(
   None=Divide(A=Distance(A=@n18, B=@n19), B=ScalarParameter() {defaultValue=1400, group=Scale By Distance, parameterName=Distance Scale})
   Min=ScalarParameter() {group=Scale By Distance, parameterName=DistanceScaleMinClamp}
   Max=ScalarParameter() {defaultValue=50, group=Scale By Distance, parameterName=DistanceScaleMaxClamp}
  ) {maxDefault=10}
  n21 = ScalarParameter() {defaultValue=1, group=WPO, parameterName=WidthTop}
  n22 = ScalarParameter() {defaultValue=1, group=WPO, parameterName=LengthNoiseTile}
  n23 = Multiply(A=TextureCoordinate(), B=@n22)
  n24 = Multiply(A=TextureCoordinate() {vTiling=2}, B=@n22)
  n25 = Add(
   A=TextureSample(UVs=Panner(Coordinate=@n23) {speedY=1}) {texture=/Game/Global/Textures/Default/Masks_and_Gradients/CylGradient}.G
   B=TextureSample(UVs=Panner(Coordinate=@n24) {speedY=1}) {texture=&CylGradient}.G
  )
  n26 = TextureCoordinate() {vTiling=6}
  n27 = TextureSample(UVs=Panner(Coordinate=Multiply(A=@n26, B=@n22)) {speedY=1}) {texture=&CylGradient}
  n28 = MaterialFunctionCall(
   WidthBase=Add(A=@n20, B=Multiply(A=@n16.Thickness, B=@n17))
   WidthTip=Add(A=Multiply(A=@n16.Thickness, B=@n21), B=@n20)
   UVs for Thickness=Multiply(A=Add(A=@n25, B=@n27.G), B=ScalarParameter() {defaultValue=2.5, group=WPO, parameterName=LengthNoise})
  ) {materialFunction=&SplineThicken}
  n29 = DynamicParameter() {defaultValue=<rgba 1 0 1 1>, paramNames [Thickness,Random,PanSpeed,PowerGrow]}
  n30 = ScalarParameter() {defaultValue=1, group=WPO, parameterName=WobbleNoiseTile}
  n31 = Multiply(A=TextureCoordinate() {uTiling=0.25, vTiling=0.25}, B=@n30)
  n32 = Time()
  n33 = ScalarParameter() {defaultValue=8, group=WPO, parameterName=StepSpeed}
  n34 = Multiply(A=@n32, B=@n33)
  n35 = StaticSwitchParameter(
   True=Divide(A=Round(None=@n34), B=@n33)
   False=Multiply(A=@n32, B=DynamicParameter() {defaultValue=<rgba 1 0 1 1>, paramNames [Thickness,Random,PanSpeed,PowerGrow]}.PanSpeed)
  ) {defaultValue=true, group=WPO, parameterName=UseStaccatoTiming}
  n36 = TextureSampleParameter2D(UVs=Panner(Coordinate=Add(A=@n29.Random, B=@n31), Time=@n35) {speedX=0.25, speedY=0.25}) {group=WPO, parameterName=Wobble Noise, samplerType=SAMPLERTYPE_Normal, texture=/Game/Global/Textures/Default/Noise/T_Normal_Noise01}
  n37 = Multiply(A=TextureCoordinate(), B=@n30)
  n38 = DynamicParameter() {defaultValue=<rgba 1 0 1 1>, paramNames [Thickness,Random,PanSpeed,PowerGrow]}
  n39 = TextureSampleParameter2D(UVs=Panner(Coordinate=Add(A=@n37, B=@n38.Random), Time=@n35) {speedX=0.2, speedY=0.2}) {group=WPO, parameterName=Wobble Noise, samplerType=SAMPLERTYPE_Normal, texture=&T_Normal_Noise01}
  n40 = Add(A=Multiply(A=@n36.RGB, B=VectorParameter() {group=WPO, parameterName=WobbleMult, defaultValue=<rgba 0.2 0.2 0.5 1>}.RGB), B=@n39.RGB)
  n41 = Multiply(A=ComponentMask(None=@n40) {r=true}) {constB=2}
  n42 = Multiply(A=ComponentMask(None=@n40) {g=true}) {constB=2}
  n43 = ComponentMask(None=@n40) {b=true}
  n44 = AppendVector(A=AppendVector(A=Subtract(A=@n41), B=Subtract(A=@n42)), B=Subtract(A=Multiply(A=@n43) {constB=2}))
  n45 = Transform(None=@n44) {transformSourceType=TRANSFORMSOURCE_World, transformType=TRANSFORM_Local}
  n46 = Transform(None=@n44) {transformSourceType=TRANSFORMSOURCE_Local}
  n47 = VectorParameter() {group=WPO, parameterName=WobbleOffset, defaultValue=<rgba 1 1 50 1>}
  n48 = Transform(None=@n47.RGB) {transformSourceType=TRANSFORMSOURCE_World, transformType=TRANSFORM_Local}
  n49 = Transform(None=@n47.RGB) {transformSourceType=TRANSFORMSOURCE_Local}
  n50 = Multiply(
   A=Multiply(
    A=StaticSwitchParameter(True=@n45, False=@n46) {group=WPO, parameterName=Use Proper Transforms}
    B=StaticSwitchParameter(True=@n48, False=@n49) {group=WPO, parameterName=Use Proper Transforms}
   )
   B=ScalarParameter() {defaultValue=0.1, group=WPO, parameterName=NoiseScale}
  )
  n51 = Clamp(None=Multiply(A=@n10) {constB=10}) {minDefault=0.05}
  n52 = ComponentMask(None=TextureCoordinate()) {g=true}
  n53 = Power(Base=Multiply(A=OneMinus(None=@n52), B=@n52), Exp=ScalarParameter() {defaultValue=3, group=Color, parameterName=Mask Exponent}) {constExponent=3}
  n54 = ScalarParameter() {defaultValue=50, group=Color, parameterName=Mask Scale}
  n55 = StaticSwitchParameter(True=Clamp(None=Multiply(A=@n53, B=@n54) {constB=50}), False=@n51) {parameterName=NoiseRampToggle}
  n56 = Add(
   A=@n28.WorldPositionOffset
   B=Multiply(
    A=Multiply(A=@n50, B=@n55)
    B=MaterialFunctionCall() {materialFunction=/Engine/Functions/Engine_MaterialFunctions02/WorldPositionOffset/ObjectScale}.Scale XYZ
   )
  )
  n57 = VectorParameter() {group=WPO, parameterName=PivotOffset}
  n58 = StaticSwitchParameter(
   True=Transform(None=@n57.RGB) {transformSourceType=TRANSFORMSOURCE_World, transformType=TRANSFORM_Local}
   False=Transform(None=@n57.RGB) {transformSourceType=TRANSFORMSOURCE_Local}
  ) {group=WPO, parameterName=Use Proper Transforms}
  ShadingPathSwitch(Default=Add(A=Multiply(A=@n56, B=@n51), B=@n58), Mobile=Add(A=@n28.WorldPositionOffset, B=@n58)) {desc=Migrated from FeatureLevelSwitch}
  n59 = ComponentMask(None=TextureCoordinate()) {g=true}
  n60 = DynamicParameter() {defaultValue=<rgba 1 0 1 1>, paramNames [Thickness,Random,PanSpeed,PowerGrow]}
  n61 = StaticSwitchParameter(True=Multiply(A=@n11, B=Power(Base=@n59, Exp=@n60.PowerGrow)), False=@n11) {parameterName=UseParticleAlpha_1}
  StaticSwitchParameter(True=Multiply(A=@n61, B=ParticleColor().A), False=@n61) {parameterName=UseParticleAlpha}
 ]
}
