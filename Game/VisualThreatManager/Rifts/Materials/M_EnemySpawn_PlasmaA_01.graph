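# Translucent unlit plasma material for enemy spawn rift VFX on mesh and sprite particles.
# A UV band blends inner and outer colors, optionally from dynamic parameters, tinted by particle color, with optional tip fade.
# Vertex offset uses particle-relative vector noise masked by vertex red, stepped by summed sines for staccato motion, disabled on low quality.
# Camera-distance scaling along normals and a local pivot offset add to the displacement.
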Material /Game/VisualThreatManager/Rifts/Materials/M_EnemySpawn_PlasmaA_01
graph {
 properties {
  bEnableResponsiveAA=true
  bUsedWithMeshParticles=true
  bUsedWithNiagaraMeshParticles=true
  bUsedWithParticleSprites=true
  blendMode=BLEND_Translucent
  numCustomizedUVs=4
  pixelDepthOffsetMode=PDOM_Legacy
  shadingModel=MSM_Unlit
  usedShadingModels=Unlit
 }
 nodes [
  n1 = DynamicParameter() {desc=Color Outer, parameterIndex=1, paramNames [R,G,B,A]}
  n2 = ScalarParameter() {group=Color, parameterName=Use Dynamic Parameter Colors, sliderMax=1}
  n3 = DynamicParameter() {desc=Color Inner, parameterIndex=2, paramNames [R,G,B,A]}
  n4 = ComponentMask(None=TextureCoordinate()) {g=true}
  n5 = Power(Base=Multiply(A=OneMinus(None=@n4), B=@n4), Exp=ScalarParameter() {defaultValue=3, group=Color, parameterName=Mask Exponent}) {constExponent=3}
  n6 = ScalarParameter() {defaultValue=50, group=Color, parameterName=Mask Scale}
  n7 = LinearInterpolate(
   A=LinearInterpolate(
    A=VectorParameter() {group=Color, parameterName=Color Outer, defaultValue=<rgba 0.5333 0.5333 80 1>}.RGB
    B=MaterialFunctionCall(X=@n1.R, Y=@n1.R, Z=@n1.R) {materialFunction=&MakeFloat3}.Result
    Alpha=@n2
   )
   B=LinearInterpolate(
    A=VectorParameter() {group=Color, parameterName=Color Inner, defaultValue=<rgba 28.25 6.5 50 1>}.RGB
    B=MaterialFunctionCall(X=@n3.R, Y=@n3.R, Z=@n3.R) {materialFunction=&MakeFloat3}.Result
    Alpha=@n2
   )
   Alpha=Clamp(None=Multiply(A=@n5, B=@n6) {constB=50})
  )
  n8 = MaterialFunctionCall(Float2=TextureCoordinate()) {materialFunction=&BreakOutFloat2Components}
  n9 = Clamp(None=Multiply(A=OneMinus(None=@n8.G), B=ScalarParameter() {defaultValue=5, group=Fade Tip, parameterName=OpacityScale}))
  n10 = Multiply(A=@n7, B=@n9)
  n11 = ParticleColor()
  n12 = ScalarParameter() {defaultValue=1, parameterName=Use Particle Color}
  MaterialFunctionCall(Emissive=Multiply(
    A=StaticSwitchParameter(True=@n10, False=@n7) {group=Fade Tip, parameterName=FadeTip}
    B=LinearInterpolate(B=@n11.RGB, Alpha=@n12) {constA=1}
   )) {materialFunction=&MF_FX_EmissiveAdjustments}
  Multiply(A=LinearInterpolate(B=@n11.A, Alpha=@n12) {constA=1}, B=StaticSwitchParameter(True=@n9, False=Constant() {r=1}) {parameterName=FadeTip})
  n13 = Time() {bOverride_Period=true, period=5.7675}
  n14 = DynamicParameter() {defaultValue=<rgba 10 0.03 1 1>, paramNames [Strength Scale, Tile Scale, Noise Speed, Moprh Speed]}
  n15 = Time() {bOverride_Period=true, period=8.6512}
  n16 = Add(A=Sine(None=Multiply(A=@n13, B=@n14.Noise Speed)) {period=5.7675}, B=Sine(None=Multiply(A=@n15, B=@n14.Noise Speed)) {period=8.6512})
  n17 = ScalarParameter() {defaultValue=3, parameterName=steps}
  n18 = StaticSwitchParameter(True=Ceil(None=Multiply(A=@n16, B=@n17) {constB=3}), False=Time() {bOverride_Period=true, period=1}) {defaultValue=true, parameterName=UseStaccatoTiming}
  n19 = WorldPosition()
  n20 = Multiply(
   A=Subtract(A=Add(A=@n18, B=@n19.XYZ), B=ParticlePositionWS())
   B=DynamicParameter() {defaultValue=<rgba 10 0.03 1 1>, paramNames [Strength Scale, Tile Scale, Noise Speed, Moprh Speed]}.Tile Scale
  )
  n21 = Multiply(
   A=Multiply(
    A=VectorNoise(World Position=@n20) {noiseFunction=VNF_VectorALU}
    B=DynamicParameter() {defaultValue=<rgba 10 0.03 1 1>, paramNames [Strength Scale, Tile Scale, Noise Speed, Moprh Speed]}.Strength Scale
   ) {constB=5}
   B=VertexColor().R
  )
  n22 = QualitySwitch(Default=@n21, Low=Constant(), High=@n21, Medium=@n21)
  n23 = StaticSwitchParameter(True=Multiply(A=@n9, B=@n22), False=@n22) {parameterName=FadeTip}
  n24 = Distance(A=ObjectPositionWS(), B=CameraPositionWS())
  n25 = ScalarParameter() {defaultValue=700, group=Scale By Distance, parameterName=Distance Scale}
  n26 = Clamp(
   None=Multiply(A=Divide(A=@n24, B=@n25), B=VertexNormalWS())
   Min=ScalarParameter() {group=Scale By Distance, parameterName=DistanceScaleMinClamp}
   Max=ScalarParameter() {defaultValue=10, group=Scale By Distance, parameterName=DistanceScaleMaxClamp}
  ) {maxDefault=10}
  n27 = VectorParameter() {parameterName=PivotOffset}
  Reroute(None=Add(A=Add(A=@n23, B=@n26), B=Transform(None=@n27.RGB) {transformSourceType=TRANSFORMSOURCE_Local}))
 ]
}
