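# Material function that samples a 2D texture laid out as a grid of frames as a tiling pseudo-volume texture, outputting RGB, alpha and a debug-style slice UV offset.
# Position, world position by default, is divided by Scale; wrapped XY maps into a cell, while Z times the squared X frame count picks the slice.
# Two adjacent slices are sampled at a fixed mip level and blended by fractional Z; the DDX/DDY nodes are unconnected.
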MaterialFunction /Game/Packages/Fortnite_UniqueMaterials_1/MaterialFunctions/VolumeTextureFunction
graph {nodes [
  n1 = FunctionInput() {
   bUsePreviewValueAsDefault=true
   description=Acts as a divisor of the position input.
   inputName=Scale
   sortPriority=2
   previewValue=<100 100 100 1>
  }
  n2 = Divide(
   A=FunctionInput(Preview=WorldPosition().XYZ) {bUsePreviewValueAsDefault=true, description=The volume texture to use., inputName=Position, sortPriority=1}
   B=@n1
  )
  n3 = Divide(A=ComponentMask(None=@n2) {g=true, r=true}) {constB=8}
  DDX(Value=@n3)
  DDY(Value=@n3)
  n4 = FunctionInput() {
   bUsePreviewValueAsDefault=true
   description=The number of frames in the volume texture.
   inputName=Num Frames
   inputType=FunctionInput_Vector2
   sortPriority=3
   previewValue=<12 12 0 1>
  }
  n5 = Multiply(A=@n4)
  n6 = Multiply(A=@n5)
  n7 = Divide(A=Frac(None=ComponentMask(None=@n2) {g=true, r=true}), B=@n6) {constB=12}
  n8 = MaterialFunctionCall(Float2=@n4) {materialFunction=&BreakOutFloat2Components}
  n9 = Multiply(A=ComponentMask(None=@n2) {b=true}, B=Multiply(A=@n8.R, B=@n8.R))
  n10 = Floor(None=@n9)
  n11 = Divide(
   A=MaterialFunctionCall!(1D Index=@n10, Cells XY=@n5) {materialFunction=/Engine/Functions/Engine_MaterialFunctions01/Coordinates/1Dto2DIndex}.2D Index
   B=@n6
  ) {constB=12}
  n12 = FunctionInput(Preview=TextureObject() {texture=&Black}) {bUsePreviewValueAsDefault=true, description=The volume texture to use., inputName=Texture Object, inputType=FunctionInput_Texture2D}
  n13 = Constant()
  n14 = TextureSample!(UVs=Add(A=@n7, B=@n11), Tex=@n12, Level=@n13, Apply View MipBias) {constMipValue=0, mipValueMode=TMVM_MipLevel, samplerType=SAMPLERTYPE_LinearColor}
  n15 = MaterialFunctionCall(1D Index=Add(A=@n10), Cells XY=@n5) {materialFunction=&1Dto2DIndex}
  n16 = TextureSample(UVs=Add(A=@n7, B=Divide(A=@n15.2D Index, B=@n6) {constB=12}), Tex=@n12, Level=@n13) {constMipValue=0, mipValueMode=TMVM_MipLevel, samplerType=SAMPLERTYPE_LinearColor}
  n17 = Frac(None=@n9) {desc=Z Blend Phase}
  FunctionOutput(None=LinearInterpolate(A=@n14.RGB, B=@n16.RGB, Alpha=@n17) {desc=Blend Z frames}) {outputName=Result RGB}
  FunctionOutput(None=@n11) {outputName=UV 1, sortPriority=1}
  FunctionOutput(None=LinearInterpolate(A=@n14.A, B=@n16.A, Alpha=@n17)) {outputName=Result Alpha, sortPriority=1}
 ]}
