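# Volumetric fog material for rift particle effects, using the Volume domain with additive blending.
# A sphere mask from particle position and radius shapes each puff, multiplied by particle alpha and noise remapped between Low and High.
# The noise samples a curl-noise pseudo-volume texture at world position, panned over time along an optionally local-space direction.
# Separate panning vector-noise and swirl-flowmap branches do not feed this masked result.
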Material /Game/VisualThreatManager/Rifts/Materials/M_RiftVolumeFog_01
graph {
 properties {
  bUsedWithParticleSprites=true
  blendMode=BLEND_Additive
  materialDomain=MD_Volume
  pixelDepthOffsetMode=PDOM_Legacy
  translucencyLightingMode=TLM_VolumetricPerVertexNonDirectional
  usedShadingModels=Default Lit
 }
 nodes [
  n1 = VectorParameter() {parameterName=PannerXYZ, defaultValue=<rgba 15 -27 -90 0>}
  n2 = Time() {period=1}
  n3 = Divide(A=Add(A=Multiply(A=@n1.RGB, B=@n2), B=WorldPosition().XYZ), B=VectorParameter() {parameterName=Scale, defaultValue=<rgba 600 600 500 0>}.RGB)
  ConstantBiasScale(None=ComponentMask(None=VectorNoise(World Position=@n3) {noiseFunction=VNF_GradientALU}) {b=true, g=true, r=true})
  n4 = ScalarParameter() {defaultValue=0.02, parameterName=TopFlowSpeedHoriz}
  n5 = Time()
  n6 = ComponentMask(None=WorldPosition() {worldPositionShaderOffset=WPT_ExcludeAllShaderOffsets}.XYZ) {g=true, r=true}
  n7 = Multiply(A=Abs(None=ScalarParameter() {defaultValue=2500, parameterName=TopFlowTile})) {constB=-1}
  n8 = TextureSample(UVs=Add(A=Multiply(A=@n4, B=@n5), B=Divide(A=@n6, B=@n7))) {samplerType=SAMPLERTYPE_LinearColor, texture=/Game/VisualThreatManager/StormVisuals/LargeClouds/T_Swirl_Flowmap}
  Multiply(A=ScalarParameter() {defaultValue=0.25, parameterName=TopFlowAmt}, B=Subtract(A=ComponentMask(None=@n8.RGB) {g=true, r=true}) {constB=0.5})
  n9 = WorldPosition()
  n10 = ParticlePositionWS()
  n11 = ParticleRadius()
  n12 = ScalarParameter() {defaultValue=0.4, parameterName=Hardness}
  n13 = ScalarParameter() {defaultValue=0.1, parameterName=Low}
  n14 = MaterialFunctionCall(X=@n13, Y=@n13, Z=@n13) {materialFunction=&MakeFloat3}
  n15 = ScalarParameter() {defaultValue=1, parameterName=High}
  n16 = MaterialFunctionCall(X=@n15, Y=@n15, Z=@n15) {materialFunction=&MakeFloat3}
  n17 = WorldPosition()
  n18 = Normalize(VectorInput=VectorParameter() {parameterName=Curl pan Dir, defaultValue=<rgba 0 0 -1 0>}.RGB)
  n19 = Multiply(
   A=StaticSwitch(
    True=Transform(None=@n18) {transformSourceType=TRANSFORMSOURCE_Local}
    False=@n18
    Value=StaticBoolParameter() {parameterName=Transform Curl to World Space}
   )
   B=MaterialFunctionCall(Speed=ScalarParameter() {defaultValue=40, parameterName=Curl pan Speed}) {materialFunction=&TimeWithSpeedVariable}.Time
  )
  n20 = ConstantBiasScale(None=MaterialFunctionCall!(
    Texture Object=TextureObject() {samplerType=SAMPLERTYPE_LinearColor, texture=/Game/TimeOfDay/VolumeFog/T_CurlNoise}
    Position=Add(A=@n17.XYZ, B=@n19)
    Scale=ScalarParameter() {defaultValue=5000, parameterName=size}
    Num Frames=Constant() {r=16}
   ) {materialFunction=/Game/Packages/Fortnite_UniqueMaterials_1/MaterialFunctions/VolumeTextureFunction}.Result RGB) {bias=-0.5, scale=2}
  Multiply(
   A=ParticleColor().A
   B=Multiply(
    A=SphereMask(A=@n9.XYZ, B=@n10, Radius=@n11, Hardness=@n12) {hardnessPercent=0}
    B=LinearInterpolate(A=@n14.Result, B=@n16.Result, Alpha=@n20)
   )
  )
 ]
}
