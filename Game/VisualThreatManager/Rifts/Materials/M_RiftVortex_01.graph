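# Masked, dithered rift vortex material that renders an animated, twisting cloud mesh with timed lightning strikes.
# Two panning cloud layers drive color, normals and displacement, with Fresnel blending core and cloud colors plus a fake view-space highlight.
# World position offset curves, twists and noise-wobbles the mesh along Z, correcting normals for the rotation.
# Timestamp functions gate a flickering sphere-masked lightning glow and bloat, and separately fade the Fresnel-based opacity mask.
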Material /Game/VisualThreatManager/Rifts/Materials/M_RiftVortex_01
graph {
 properties {
  bEnableResponsiveAA=true
  bUsedWithMeshDeformer=true
  bUsedWithSkeletalMesh=true
  bUsedWithSplineMeshes=true
  blendMode=BLEND_Masked
  ditherOpacityMask=true
  opacityMaskClipValue=0.5
  pixelDepthOffsetMode=PDOM_Legacy
  translucencyLightingMode=TLM_VolumetricPerVertexDirectional
  translucentShadowDensityScale=2
  usedShadingModels=Default Lit
 }
 nodes [
  n1 = VectorParameter() {group=Color, parameterName=Core, defaultValue=<rgba 0.1322 0.0366 0.61 1>}
  n2 = Multiply(
   A=TextureCoordinate()
   B=AppendVector(A=ScalarParameter() {defaultValue=1, parameterName=Tile Master U}, B=ScalarParameter() {defaultValue=1, parameterName=Tile Master V})
  )
  n3 = Multiply(A=ScalarParameter() {defaultValue=3, group=Clouds A, parameterName=UA}, B=@n2)
  n4 = Frac(None=Multiply(A=Time(), B=ScalarParameter() {defaultValue=0.1304, group=Clouds A, parameterName=Panner Speed UA}))
  n5 = Multiply(A=ScalarParameter() {defaultValue=0.2788, group=Clouds A, parameterName=VA}, B=@n2)
  n6 = Frac(None=Multiply(A=Time(), B=ScalarParameter() {defaultValue=0.1722, group=Clouds A, parameterName=Panner Speed VA}))
  n7 = AppendVector(
   A=ComponentMask(None=Panner(Coordinate=@n3, Time=@n4) {bFractionalPart=true, desc="-0.093667, 0.297", speedX=1}) {r=true}
   B=ComponentMask(None=Panner(Coordinate=@n5, Time=@n6) {bFractionalPart=true, desc="-0.093667, 0.297", speedY=1}) {g=true}
  )
  n8 = Multiply(A=@n2, B=ScalarParameter() {defaultValue=2, group=Clouds B, parameterName=UB})
  n9 = Frac(None=Multiply(A=Time(), B=ScalarParameter() {defaultValue=0.1975, group=Clouds B, parameterName=Panner Speed UB}))
  n10 = Multiply(A=ScalarParameter() {defaultValue=1, group=Clouds B, parameterName=VB}, B=@n2)
  n11 = Frac(None=Multiply(A=Time(), B=ScalarParameter() {defaultValue=0.101, group=Clouds B, parameterName=Panner Speed VB}))
  n12 = AppendVector(
   A=ComponentMask(None=Panner(Coordinate=@n8, Time=@n9) {bFractionalPart=true, desc="0.117, 0.3114", speedX=1}) {r=true}
   B=ComponentMask(None=Panner(Coordinate=@n10, Time=@n11) {bFractionalPart=true, desc="0.117, 0.3114", speedY=1}) {g=true}
  )
  n13 = Divide(A=Add(
    A=TextureSample(UVs=@n7) {texture=&2ndInfectedMapLinear}.RGB
    B=TextureSample(UVs=@n12) {texture=&2ndInfectedMapLinear}.RGB
   ))
  n14 = ScalarParameter() {defaultValue=3, parameterName=Clouds Contrast}
  n15 = LinearInterpolate(
   A=VectorParameter() {group=Color, parameterName=CloudDarkColor, defaultValue=<rgba 0.0279 0.0105 0.08 1>}.RGB
   B=VectorParameter() {group=Color, parameterName=CloudLightColor, defaultValue=<rgba 0.1079 0.0605 0.25 1>}.RGB
   Alpha=Multiply(A=Power(Base=@n13, Exp=@n14), B=ScalarParameter() {defaultValue=3, parameterName=Clouds Brightness})
  )
  n16 = ScalarParameter() {defaultValue=2, group=Fresnel, parameterName=Fresnel Cloud Brightness}
  n17 = ScalarParameter() {defaultValue=3, group=Fresnel, parameterName=Fresnel Cloud Exponent}
  n18 = ScalarParameter() {defaultValue=2, group=Fresnel, parameterName=Fresnel Core Brightness}
  n19 = ScalarParameter() {defaultValue=1, group=Fresnel, parameterName=Fresnel Core Exponent}
  n20 = Constant3Vector() {constant=<rgba 0 0 1 0>}
  n21 = Constant3Vector() {constant=<rgba 0 0 1 0>}
  n22 = WorldPosition() {worldPositionShaderOffset=WPT_ExcludeAllShaderOffsets}
  n23 = VectorParameter() {parameterName=Pivot position, defaultValue=<rgba 0 0 0 1>}
  n24 = Subtract(A=@n22.XYZ, B=@n23.RGB)
  n25 = Constant3Vector() {constant=<rgba 0 0 1 0>}
  n26 = Multiply(
   A=ScalarParameter() {parameterName=curve along axis amount}
   B=Divide(A=DotProduct(A=@n24, B=@n25), B=ScalarParameter() {defaultValue=275, parameterName=Total Model height})
  )
  n27 = RotateAboutAxis(NormalizedRotationAxis=@n21, RotationAngle=@n26, PivotPoint=@n23.RGB, Position=@n22.XYZ)
  n28 = Add(A=@n27, B=@n22.XYZ) {desc=Curved world space}
  n29 = VectorParameter() {parameterName=Pivot position, defaultValue=<rgba 0 0 0 1>}
  n30 = Constant3Vector() {constant=<rgba 0 0 1 0>}
  n31 = DotProduct(A=Subtract(A=@n28, B=@n29.RGB), B=@n30)
  n32 = Divide(A=@n31, B=ScalarParameter() {defaultValue=275, parameterName=Total Model height})
  n33 = Multiply(A=@n32, B=ScalarParameter() {defaultValue=0.5, parameterName="Constant Rotation Amount (.5)"})
  n34 = MaterialFunctionCall(
   Rotation Axis=@n20
   Rotation Angle=@n33
   World Space Vertex Normals=MaterialFunctionCall(Rotation Axis=@n21, Rotation Angle=@n26, World Space Vertex Normals=VertexNormalWS()) {materialFunction=&FixRotateAboutAxisNormals}.New World Space Vertex Normal
  ) {materialFunction=&FixRotateAboutAxisNormals}
  n35 = Clamp(None=Add(
    A=MaterialFunctionCall(Brightness=@n16, Exponent=@n17) {materialFunction=&MF_Fresnel_FX}.Fresnel
    B=MaterialFunctionCall(Brightness=@n18, Exponent=@n19, Vertex / Pixel Normal=@n34.New World Space Vertex Normal) {materialFunction=&MF_Fresnel_FX}.Fresnel
   ))
  n36 = Transform(None=PixelNormalWS()) {transformSourceType=TRANSFORMSOURCE_World, transformType=TRANSFORM_View}
  n37 = Normalize(VectorInput=VectorParameter() {group=Fake Highlight, parameterName=Fresnel Highlight Vector, defaultValue=<rgba 0.3 0.6 0 0>}.RGB)
  n38 = Power(
   Base=Clamp(None=DotProduct(A=@n36, B=@n37))
   Exp=ScalarParameter() {defaultValue=1.7, group=Fake Highlight, parameterName=Fresnel Highlight Exponent}
  )
  n39 = VectorParameter() {group=Fake Highlight, parameterName=Fresnel Highlight Color, defaultValue=<rgba 0.5007 0.38 4 0>}
  n40 = Multiply(A=@n31, B=@n30)
  n41 = VectorParameter() {parameterName=XY pan speed, defaultValue=<rgba -0.05 -0.05 0 1>}
  n42 = Add(
   A=Multiply(A=@n32, B=ScalarParameter() {defaultValue=0.25, parameterName=PatternTilingFrequency})
   B=Multiply(A=ComponentMask(None=@n41.RGB) {g=true, r=true}, B=Time())
  )
  n43 = Multiply(
   A=ConstantBiasScale(None=TextureSample(UVs=@n42) {samplerType=SAMPLERTYPE_LinearColor, texture=&HighContrastNoise}.RGB) {bias=-0.5, scale=2}
   B=ScalarParameter() {defaultValue=18, parameterName=Axis Offset Amount}
  ) {constB=10}
  n44 = Add(A=RotateAboutAxis(NormalizedRotationAxis=@n20, RotationAngle=@n33, PivotPoint=Add(A=@n29.RGB, B=@n40), Position=Add(A=@n43, B=@n28)), B=@n27)
  n45 = SphereMask(
   A=Add(A=@n44, B=WorldPosition() {worldPositionShaderOffset=WPT_ExcludeAllShaderOffsets}.XYZ)
   B=VectorParameter() {group=Lightning Strike, parameterName=Lightning Center, defaultValue=<rgba 0 0 127 0>}.RGB
   Radius=ScalarParameter() {defaultValue=180, group=Lightning Strike, parameterName=Lightning Radius}
   Hardness=ScalarParameter() {defaultValue=0.15, group=Lightning Strike, parameterName=Lightning Hardness}
  )
  n46 = Multiply(A=@n45, B=OneMinus(None=@n13)) {constA=1}
  n47 = Multiply(A=Time(), B=ScalarParameter() {defaultValue=1, group=Lightning Strike, parameterName=Lightning Flicker Speed})
  n48 = Add(
   A=ConstantBiasScale(None=Sine(None=@n47) {period=0.0766}) {bias=1.9, scale=0.11}
   B=ConstantBiasScale(None=Sine(None=@n47) {period=0.1449}) {bias=1.75, scale=0.221}
  )
  n49 = Multiply(
   A=Power(Base=Multiply(A=@n46, B=@n48), Exp=ScalarParameter() {defaultValue=3, group=Lightning Strike, parameterName=LightningEmissivePower})
   B=VectorParameter() {group=Lightning Strike, parameterName=Lightning Color, defaultValue=<rgba 48.9584 12.5 100 1>}.RGB
  )
  n50 = MaterialFunctionCall(
   TimestampA=ScalarParameter() {group=Lightning Strike, parameterName=TimestampA}
   TransitionLength_In=ScalarParameter() {defaultValue=0.2, group=Lightning Strike, parameterName=TimestampLengthIn}
   TimestampB=ScalarParameter() {group=Lightning Strike, parameterName=TimestampB}
   TransitionLength_Out=ScalarParameter() {defaultValue=0.2, group=Lightning Strike, parameterName=TimestampLengthOut}
   Transition In Only?=StaticBoolParameter() {group=Lightning Strike, parameterName=Transition In Only?}
  ) {materialFunction=&MF_Timestamp_InOut_01}
  Add(
   A=Multiply(A=LinearInterpolate(A=@n1.RGB, B=@n15, Alpha=@n35), B=ScalarParameter() {defaultValue=0.01, parameterName=Emissive Boost})
   B=Add(A=Multiply(A=@n38, B=@n39.RGB), B=Multiply(A=@n49, B=@n50.Linear))
  )
  n51 = Multiply(
   A=Multiply(A=ConstantBiasScale(None=@n13) {bias=-0.5, scale=2}, B=@n34.New World Space Vertex Normal)
   B=ScalarParameter() {defaultValue=10, group=WPO, parameterName=WPO Strength}
  )
  n52 = ComponentMask(None=TextureCoordinate()) {g=true}
  n53 = ScalarParameter() {defaultValue=0.5, parameterName=Bloat Jiggle}
  n54 = Multiply(
   A=Power(Base=@n45, Exp=ScalarParameter() {defaultValue=4, parameterName=Bloat Radius Power})
   B=Multiply(A=@n50.Linear, B=ScalarParameter() {defaultValue=600, parameterName=Bloat})
  )
  n55 = Add(A=Multiply(A=Multiply(A=@n48, B=@n53), B=@n54), B=@n54)
  Add(A=@n44, B=Add(A=Multiply(A=@n51, B=@n52), B=Multiply(A=@n34.New World Space Vertex Normal, B=@n55)))
  n56 = Add(
   A=TextureSample(UVs=@n7) {samplerType=SAMPLERTYPE_Normal, texture=/Game/Global/Textures/Default/Tileable_Patterns/2ndInfectedMapNormal}.RGB
   B=TextureSample(UVs=@n12) {samplerType=SAMPLERTYPE_Normal, texture=&2ndInfectedMapNormal}.RGB
  )
  n57 = Constant3Vector() {constant=<rgba 0.5 0.5 1 0>}
  MaterialFunctionCall(BaseNormal=Normalize(VectorInput=@n34.New Tangent Space Vertex Normal), AdditionalNormal=Normalize(VectorInput=Multiply(A=@n56, B=@n57))) {materialFunction=&BlendAngleCorrectedNormals}
  ScalarParameter() {defaultValue=0.02, parameterName=Specular}
  ScalarParameter() {defaultValue=1, parameterName=Roughness}
  ScalarParameter() {defaultValue=4.5, parameterName=TestTime}
  ScalarParameter() {defaultValue=0.01, parameterName=Metallic}
  n58 = ScalarParameter() {parameterName=TimestampAFade}
  n59 = ScalarParameter() {defaultValue=1, parameterName=TransitionLength_InFade}
  n60 = ScalarParameter() {defaultValue=1000000000, parameterName=TimestampBFade}
  n61 = ScalarParameter() {defaultValue=1, parameterName=TransitionLength_OutFade}
  n62 = StaticBool()
  Subtract(
   A=@n35
   B=OneMinus(None=MaterialFunctionCall(TimestampA=@n58, TransitionLength_In=@n59, TimestampB=@n60, TransitionLength_Out=@n61, Transition In Only?=@n62) {materialFunction=&MF_Timestamp_InOut_01}.Linear)
  ) {constB=0}
 ]
}
