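# Unlit additive particle material for a spark sprite in the rift spawn effect, usable with Niagara and Cascade sprites.
# The spark texture's red channel is tinted by particle color and scaled by particle alpha.
# That result is multiplied by itself, so color and fade both respond quadratically, then passed through MF_FX_EmissiveAdjustments as emissive.
# It renders before depth of field with responsive AA enabled.
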Material /Game/VisualThreatManager/Rifts/Spark_mat_SPAWNEFFECT
graph {
 properties {
  bEnableResponsiveAA=true
  bUsedWithNiagaraSprites=true
  bUsedWithParticleSprites=true
  blendMode=BLEND_Additive
  pixelDepthOffsetMode=PDOM_Legacy
  shadingModel=MSM_Unlit
  translucencyPass=MTP_BeforeDOF
  usedShadingModels=Unlit
 }
 nodes [
  n1 = ParticleColor()
  n2 = Multiply(A=Multiply(A=TextureSample() {texture=/Game/VisualThreatManager/Rifts/Textures/testSpark}.R, B=@n1.RGB), B=@n1.A)
  MaterialFunctionCall(Emissive=Multiply(A=@n2, B=@n2)) {materialFunction=&MF_FX_EmissiveAdjustments}
 ]
}
